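# Material for the NunyaFlare cosmetic vehicle wheel, adding animated lava and pulsing glow to rim and tire on the simple wheel parent.
# Panning noise distorts a cloud texture for tinted lava, while a sine-cosine product drives the pulse.
# RGB masks and a rim Fresnel shape the glows, with red vertex color blending rim into another emissive layer.
# Modified base color and emissive attributes pass through vehicle damage and de-rez transition functions.
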Material /VehicleCosmetics/Wheels/NunyaFlare/Materials/M_Wheel_NunyaFlare
graph {
 properties {
  bAutomaticallySetUsageInEditor=false
  bUseMaterialAttributes=true
  bUsedWithMeshDeformer=true
  bUsedWithSkeletalMesh=true
  pixelDepthOffsetMode=PDOM_Legacy
  usedShadingModels=Default Lit
 }
 nodes [
  Add()
  Constant()
  Constant()
  LinearInterpolate()
  n1 = GetMaterialAttributes(None=MaterialFunctionCall() {materialFunction=&MF_WheelParent_Simple}.MaterialAttributes) {attributeGetTypes [#1, #2]}
  n2 = Add(A=@n1.Emissive Color, B=NamedRerouteUsage())
  n3 = NamedRerouteUsage()
  n4 = OneMinus(None=@n3)
  n5 = LinearInterpolate(A=Add(A=LinearInterpolate(B=@n1.Base Color, Alpha=@n4), B=NamedRerouteUsage()), B=NamedRerouteUsage(), Alpha=@n3)
  MaterialFunctionCall(In=MaterialFunctionCall(Material=SetMaterialAttributes!(MaterialAttributes=@n1.MaterialAttributes, Emissive Color=@n2, Base Color=@n5) {attributeSetTypes [#1, #2]}) {materialFunction=&MF_BaseVehicle_Damage}.Result) {materialFunction=&MF_VehicleDeRez_Transition}
  n6 = TextureCoordinate() {uTiling=2, vTiling=2}
  n7 = TextureSample(UVs=Panner(Coordinate=@n6) {speedY=-0.25}) {texture=/VehicleCosmetics/Wheels/TankFlare/Textures/T_Tank_Flare_Noise_RGB}
  n8 = Constant() {r=0.1}
  n9 = Panner(Coordinate=Add(A=@n6, B=Multiply(A=@n7.B, B=@n8))) {speedX=-0.4, speedY=0.1}
  n10 = Subtract(
   A=Clamp(None=TextureSample(UVs=@n9) {texture=/VehicleCosmetics/Wheels/_SharedTextures/FX/Clouds_RGB_Soft_03_T}.R)
   B=Clamp(None=ConstantBiasScale(None=@n7.B) {bias=-2, scale=1})
  )
  n11 = Add(A=Clamp(None=Max(A=@n10, B=@n10)), B=Constant() {r=0.1})
  n12 = VectorParameter() {group=5_Special Effects, parameterName=LavaColor, defaultValue=<rgba 1 0.0369 0 1>}
  NamedRerouteDeclaration(None=Multiply(A=Multiply(A=@n11, B=@n12.RGB), B=ScalarParameter() {defaultValue=4, group=5_Special Effects, parameterName=LavaIntensity})) {name=Lava, nodeColor=<rgba 0 0.6953 1 1>}
  n13 = Multiply(A=Time(), B=ScalarParameter() {defaultValue=0.5, group=5_Special Effects, parameterName=PulseSpeed})
  n14 = Sine(None=@n13)
  n15 = Cosine(None=Multiply(A=@n13, B=Constant() {r=0.5}))
  n16 = Constant()
  NamedRerouteDeclaration(None=Max(A=Clamp(None=@n16), B=Multiply(A=@n14, B=@n15))) {name=Pulse, nodeColor=<rgba 0.6641 0 1 1>}
  NamedRerouteDeclaration(None=VectorParameter() {group=5_Special Effects, parameterName=AccentColorRIM, defaultValue=<rgba 1 0.0369 0 1>}.RGB) {name=AccentColorRim, nodeColor=<rgba 0 1 0.3438 1>}
  NamedRerouteDeclaration(None=VectorParameter() {group=5_Special Effects, parameterName=PulseColorRIM, defaultValue=<rgba 1 0 0 1>}.RGB) {name=PulseColor, nodeColor=<rgba 0 1 0.3438 1>}
  n17 = NamedRerouteUsage()
  n18 = TextureSampleParameter2D() {parameterName=Masks_Tire_E, texture=/VehicleCosmetics/Wheels/NunyaFlare/Textures/T_NunyaFlare_Tire_RGB}
  n19 = ScalarParameter() {defaultValue=1, group=5_Special Effects, parameterName=TirePulseIntensity}
  n20 = Multiply(A=NamedRerouteUsage(), B=@n19)
  n21 = NamedRerouteUsage()
  n22 = Add(
   A=Multiply(A=Multiply(A=@n17, B=@n18.B), B=ScalarParameter() {defaultValue=0.8, group=5_Special Effects, parameterName=TireDetailGlowIntensity})
   B=Multiply(A=@n18.B, B=Multiply(A=@n20, B=@n21))
  )
  n23 = NamedRerouteUsage()
  n24 = NamedRerouteUsage()
  NamedRerouteDeclaration(None=Multiply(A=LinearInterpolate(A=@n22, B=@n23, Alpha=@n18.R), B=Add(A=@n19, B=@n24))) {name=TireE, nodeColor=<rgba 0 0.2969 1 1>}
  n25 = Constant()
  n26 = NamedRerouteUsage()
  n27 = TextureSampleParameter2D() {
   group=5_Special Effects
   parameterName=Masks_Rim_E
   samplerSource=SSM_Wrap_WorldGroupSettings
   texture=/VehicleCosmetics/Wheels/NunyaFlare/Textures/T_NunyaFlare_Rim_RGB
  }
  n28 = Multiply(A=VectorParameter() {group=5_Special Effects, parameterName=CustomColorRIM, defaultValue=<rgba 1 0.0369 0 1>}.RGB, B=@n27.G)
  n29 = Constant() {r=1.5}
  n30 = Multiply(A=@n29, B=NamedRerouteUsage())
  n31 = Add(A=@n29, B=@n30)
  n32 = Multiply(A=NamedRerouteUsage(), B=@n27.G)
  n33 = TextureSampleParameter2D() {
   group=5_Special Effects
   parameterName=Normal_Rim
   samplerSource=SSM_Wrap_WorldGroupSettings
   samplerType=SAMPLERTYPE_Normal
   texture=/VehicleCosmetics/Wheels/NunyaFlare/Textures/T_NunyaFlare_Rim_N
  }
  n34 = Multiply(
   A=MaterialFunctionCall!(
    Input Normal=@n33.RGB
    Input Sharpness=ScalarParameter() {defaultValue=1.3, group=5_Special Effects, parameterName=FresnelFeather}
    Input Two Sided
   ) {materialFunction=/VehicleCosmetics/Wheels/CarCar_Holographic/Materials/MF_Fresnel_Advanced}.Edge
   B=NamedRerouteUsage()
  )
  n35 = Add(
   A=Add(A=LinearInterpolate(A=@n25, B=@n26, Alpha=@n27.R), B=Multiply(A=@n28, B=@n31))
   B=Add(A=Multiply(A=@n30, B=@n32), B=Multiply(A=@n34, B=@n30))
  )
  n36 = VertexColor()
  NamedRerouteDeclaration(None=LinearInterpolate(A=NamedRerouteDeclaration(None=@n35) {name=Rim, nodeColor=<rgba 0.7578 1 0 1>}, B=NamedRerouteUsage(), Alpha=@n36.R)) {name=Emissive, nodeColor=<rgba 0.6172 1 0 1>}
  NamedRerouteDeclaration(None=@n36.R) {name=Red_VC, nodeColor=<rgba 0.9219 0 1 1>}
  OneMinus(None=@n27.R)
 ]
}
