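# Masked LOD material for the animated, glowing lava on the NunyaFlare vehicle wheel, usable on skeletal meshes.
# A rotating spiral texture warps rotating noise UVs, and that noise both distorts and subtracts from a spinning cloud texture, which is tinted by LavaColor and LavaIntensity.
# A sine-cosine Pulse reroute is built alongside the Lava reroute for emissive modulation.
# Emissive and a texture-driven opacity mask pass through the shared vehicle damage and de-rez transition functions.
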Material /VehicleCosmetics/Wheels/NunyaFlare/Materials/M_Wheel_NunyaFlare_Lava_LOD
graph {
 properties {
  bAutomaticallySetUsageInEditor=false
  bUseMaterialAttributes=true
  bUsedWithSkeletalMesh=true
  blendMode=BLEND_Masked
  usedShadingModels=Default Lit
 }
 nodes [
  Constant()
  n1 = TextureCoordinate() {coordinateIndex=1}
  n2 = Panner(Coordinate=TextureCoordinate() {uTiling=0.25, vTiling=0.25}, Time=Constant() {r=0.75}) {speedX=0.5, speedY=0.5}
  n3 = OneMinus(None=TextureSample(UVs=Rotator(Coordinate=@n2) {speed=2}) {samplerType=SAMPLERTYPE_LinearColor, texture=/VehicleCosmetics/Wheels/NunyaFlare/Textures/Spiral_CCW}.R)
  n4 = TextureSample(UVs=Rotator(Coordinate=Add(A=@n1, B=@n3)) {speed=0.1}) {texture=/VehicleCosmetics/Wheels/TankFlare/Textures/T_Tank_Flare_Noise_RGB}
  n5 = Constant() {r=0.1}
  n6 = Rotator(Coordinate=Add(A=Multiply(A=@n4.B, B=@n5), B=TextureCoordinate()), Time=Multiply(A=Time(), B=Constant() {r=0.5})) {speed=1}
  n7 = Subtract(
   A=Clamp(None=TextureSample(UVs=@n6) {texture=/VehicleCosmetics/Wheels/_SharedTextures/FX/Clouds_RGB_Soft_03_T}.R)
   B=Clamp(None=ConstantBiasScale(None=@n4.B) {bias=-0.2, scale=1})
  )
  n8 = Max(A=@n7, B=@n7)
  n9 = Multiply(
   A=Add(A=Clamp(None=@n8), B=Constant() {r=0.15})
   B=VectorParameter() {group=5_Special Effects, parameterName=LavaColor, defaultValue=<rgba 1 0.0369 0 1>}.RGB
  )
  n10 = ScalarParameter() {defaultValue=1, group=5_Special Effects, parameterName=LavaIntensity}
  n11 = Multiply(A=NamedRerouteDeclaration(None=Multiply(A=@n9, B=@n10)) {name=Lava, nodeColor=<rgba 0 0.6953 1 1>}, B=NamedRerouteUsage())
  n12 = TextureSample() {samplerType=SAMPLERTYPE_LinearColor, texture=/VehicleCosmetics/Wheels/NunyaFlare/Textures/T_NunyaFlare_Lava_M_LOD}
  MaterialFunctionCall(In=MaterialFunctionCall(Material=SetMaterialAttributes!(MaterialAttributes, Emissive Color=@n11, Opacity Mask=@n12.RGB) {attributeSetTypes [#1, #2]}) {materialFunction=&MF_BaseVehicle_Damage}.Result) {materialFunction=&MF_VehicleDeRez_Transition}
  n13 = Multiply(A=Time(), B=ScalarParameter() {defaultValue=0.5, group=5_Special Effects, parameterName=PulseSpeed})
  n14 = Constant() {r=0.5}
  n15 = Multiply(A=Sine(None=@n13), B=Cosine(None=Multiply(A=@n13, B=@n14)))
  NamedRerouteDeclaration(None=Add(A=Clamp(None=@n15), B=Constant() {r=1})) {name=Pulse, nodeColor=<rgba 0.6641 0 1 1>}
  ScalarParameter() {defaultValue=1, parameterName=Roughness}
 ]
}
